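# Animation Blueprint for the SoapPocket backpack that procedurally offsets its pack bone to follow the body and skydiving states.
# Spine bend relative to ref pose drives the pack's additive rotation and translation each update.
# A floating-state blend selects between a mesh-copied pose with a timed parachute-close offset and a cached input pose with skydive/dive offsets, before fixed nudges.
# It also tracks SoapPocket mesh and pickaxe flags, which the AnimGraph never reads.
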AnimBlueprint /BRCosmetics/Accessories/FORT_Backpacks/Backpack_M_MED_SoapPocket/Meshes/M_MED_SoapPocket_Pack_AnimBP
target_skeleton=&/M_MED_SoapPocket_Pack_Skeleton
graphs [
 AnimGraph {nodes [
   n1 = "Transform (Modify) Bone - Bone: pack"(
    ComponentPose=Local To Component(LocalPose=Copy Pose From Mesh())
    Alpha=0.4
    Translation=Get Backpack Translation() -> (Backpack Translation:Vector)
    Rotation=Get Backpack Rotation() -> (Backpack Rotation:Rotator)
   ) {
    rotationMode=BMM_Additive
    scaleMode=BMM_Additive
    scaleSpace=BCS_BoneSpace
    translationMode=BMM_Additive
    alphaBoolBlend {blendOption=Linear}
    boneToModify {boneName=pack}
   }
   n2 = "Transform (Modify) Bone - Bone: pack"!(
    ComponentPose=@n1
    bAlphaBoolEnabled:bool=Get Parashute Closed() -> (Parashute Closed:bool)
    Translation="0,-4,0"
    Rotation="0,-5,-10"
    Scale
   ) -> (Pose) {
    alphaInputType=Bool
    rotationMode=BMM_Additive
    rotationSpace=BCS_BoneSpace
    translationMode=BMM_Additive
    translationSpace=BCS_BoneSpace
    alphaBoolBlend {blendInTime=0.1, blendOption=Linear, blendOutTime=0.3}
    alphaScaleBiasClamp {bInterpResult=true, interpSpeedDecreasing=1, interpSpeedIncreasing=1}
    boneToModify {boneName=pack}
    rotation {roll=-10, yaw=-5}
    translation {y=-4}
   }
   n3 = Component To Local(ComponentPose=@n2)
   n4 = "Transform (Modify) Bone - Bone: pack"(
    ComponentPose=Local To Component(LocalPose=Use cached pose 'Base_pose'())
    bAlphaBoolEnabled=Get bIsSkyDiving()
    Translation="0,-2,8"
    Rotation="0,0,15"
   ) {
    alphaInputType=Bool
    rotationMode=BMM_Replace
    rotationSpace=BCS_BoneSpace
    translationMode=BMM_Additive
    translationSpace=BCS_BoneSpace
    alphaBoolBlend {blendInTime=0.2, blendOption=Linear, blendOutTime=0.2}
    boneToModify {boneName=pack}
    rotation {roll=15}
    translation {y=-2, z=8}
   }
   n5 = "Transform (Modify) Bone - Bone: pack"(ComponentPose=@n4, bAlphaBoolEnabled=Get bIsSkydiveDiveMode(), Translation="0,3,-3", Rotation="0,0,-15") {
    alphaInputType=Bool
    rotationMode=BMM_Additive
    rotationSpace=BCS_BoneSpace
    translationMode=BMM_Additive
    translationSpace=BCS_BoneSpace
    alphaBoolBlend {blendInTime=0.1, blendOption=Linear, blendOutTime=0.3}
    alphaScaleBiasClamp {bInterpResult=true, interpSpeedDecreasing=1, interpSpeedIncreasing=1}
    boneToModify {boneName=pack}
    rotation {roll=-15}
    translation {y=3, z=-3}
   }
   n6 = Component To Local(ComponentPose=@n5)
   n7 = Get bIsSkydiveFloating()
   n8 = "Transform (Modify) Bone - Bone: pack"(
    ComponentPose=Local To Component(LocalPose=Blend Poses by bool(BlendPose_0=@n3, BlendPose_1=@n6, bActiveValue=@n7) {blendType=Linear})
    Translation="0,1.2,0"
    Rotation="0,0,0"
   ) {
    alphaInputType=Bool
    translationMode=BMM_Additive
    translationSpace=BCS_ParentBoneSpace
    alphaBoolBlend {blendOption=Linear}
    boneToModify {boneName=pack}
    translation {y=1.2}
   }
   n9 = "Transform (Modify) Bone - Bone: pack"(ComponentPose=@n8, Translation="0,1.6,0", Rotation="0,0,0") {
    alphaInputType=Bool
    rotationSpace=BCS_BoneSpace
    scaleSpace=BCS_BoneSpace
    translationMode=BMM_Additive
    translationSpace=BCS_ParentBoneSpace
    alphaBoolBlend {blendOption=Linear}
    boneToModify {boneName=pack}
    translation {y=1.6}
   }
   Output Pose(Result=Component To Local(ComponentPose=@n9)) {layerGroup=None}
   n10 = "Transform (Modify) Bone - Bone: pack"!(
    ComponentPose=Local To Component(LocalPose=Input Pose())
    Alpha:float
    Translation=Get Backpack Translation()
    Rotation=Get Backpack Rotation()
    Scale
   ) -> (Pose) {
    rotationMode=BMM_Additive
    scaleMode=BMM_Additive
    scaleSpace=BCS_BoneSpace
    translationMode=BMM_Additive
    alphaBoolBlend {blendOption=Linear}
    boneToModify {boneName=pack}
   }
   Save cached pose 'Base_pose'(Pose=Component To Local(ComponentPose=@n10)) {cacheName=Base_pose}
  ]}
 EventGraph {nodes [
   Event BlueprintUpdateAnimation()
   > n1 = Sequence!(execute) -> (then_0, then_1, then_2:Exec, then_3:Exec)
   n2 = GetOwnerSkeletalMeshComponent()
   n3 = BreakRotator(InRot=BreakTransform(InTransform=GetDeltaTransformFromRefPose(self=@n2, BoneName=spine_05, BaseName=spine_01)).Rotation)
   n4 = MapRangeClamped(Value=@n3.Pitch, InRangeA=30, InRangeB=-30, OutRangeA=-30, OutRangeB=30)
   n5 = MapRangeClamped(Value=@n3.Pitch, InRangeA=0, InRangeB=-30, OutRangeA=-5, OutRangeB=25)
   n6 = OR Boolean(A=Get bIsSkydiveFloating(), B=Get bIsSkyDiving())
   > Set Backpack Rotation(Backpack Rotation:Rotator=MakeRotator(Roll=MapRangeClamped(Value=@n3.Yaw, InRangeA=0, InRangeB=-30, OutRangeA=5, OutRangeB=25), Pitch=SelectFloat(A=@n5, B=@n4, bPickA=@n6))) -> (Output_Get:Rotator)
   > Set Backpack Translation(Backpack Translation:Vector=MakeVector(Y=MapRangeClamped(Value=@n3.Yaw, InRangeA=0, InRangeB=-30, OutRangeB=-6.5))) -> (Output_Get:Vector)
   n7 = Cast To FortPlayerPawn(execute=@n1.then_1, Object=TryGetPawnOwner())
   n8 = GetSkeletalMeshForPartType(self=@n7)
   > Set IsSoapPocket(IsSoapPocket:bool=MatchesWildcard(SourceString=GetDisplayName(Object=@n8), Wildcard=*M_MED_SoapPocket*, SearchCase=CaseSensitive)) -> (Output_Get:bool)
   n9 = Cast To FortPlayerAnimInstance(execute=@n1.then_2, Object=GetAnimInstance(self=GetOwnerSkeletalMeshComponent()))
   > Set isPickaxe(isPickaxe:bool=false) -> (Output_Get:bool)
   > n10 = Switch on EFortWeaponCoreAnimation!(execute, Selection:EFortWeaponCoreAnimation Enum=Get WeaponCoreAnim(self=@n9)) -> (
    Melee:Exec
    Pistol:Exec
    Shotgun:Exec
    PaperBlueprint:Exec
    Rifle:Exec
    MeleeOneHand:Exec
    MachinePistol:Exec
    RocketLauncher:Exec
    GrenadeLauncher:Exec
    GoingCommando:Exec
    AssaultRifle:Exec
    TacticalShotgun:Exec
    SniperRifle:Exec
    TrapPlacement:Exec
    ShoulderLauncher:Exec
    AbilityDecoTool:Exec
    Crossbow:Exec
    C4:Exec
    RemoteControl:Exec
    DualWield:Exec
    AR_BullPup:Exec
    AR_ForwardGrip:Exec
    MedPackPaddles:Exec
    SMG_P90:Exec
    AR_DrumGun:Exec
    Consumable_Small:Exec
    Consumable_Large:Exec
    Balloon:Exec
    MountedTurret:Exec
    CreativeTool:Exec
    ExplosiveBow:Exec
    AshtonIndigo:Exec
    AshtonChicago:Exec
    MeleeDualWield:Exec
    Unarmed:Exec
   )
   Set isPickaxe(execute=[@n10.DualWield, @n10.Melee, @n10.MeleeDualWield, @n10.MeleeOneHand], isPickaxe=true)
   Set Holding_SoapPocket_pickaxe(
    execute=@n1.then_3
    Holding_SoapPocket_pickaxe:bool=AND Boolean(A=Get SoapPocket_pickaxe() -> (SoapPocket_pickaxe:bool), B=Get isPickaxe() -> (isPickaxe:bool))
   ) -> (Output_Get:bool)
   Event OnCharacterCustomizationLoaded()
   > n11 = Branch(Condition=PickaxeItemDefinitionHasAnyMetaTag(Tags=Make Array("[0]"=(TagName="Cosmetics.MetaTags.SoapPocket")), bRequiresEquipped=false))
   > Set SoapPocket_pickaxe(SoapPocket_pickaxe:bool=true) -> (Output_Get:bool)
   Set SoapPocket_pickaxe(execute=@n11.else, SoapPocket_pickaxe=false)
   Event OnParachuteClosed()
   > Set Parashute Closed(Parashute Closed:bool=true) -> (Output_Get:bool)
   > Delay(Duration=3)
   > Set Parashute Closed(Parashute Closed=false)
   TryGetPawnOwner()
  ]}
]
variables [
 Backpack Rotation {type=Rotator, default=<rot 0 0 5>}
 Backpack Translation {type=Vector}
 Holding_SoapPocket_pickaxe {type=boolean}
 isPickaxe {type=boolean}
 IsSoapPocket {type=boolean}
 Parashute Closed {type=boolean}
 Skel Mesh {type=SkeletalMeshComponent}
 SoapPocket_pickaxe {type=boolean}
]
